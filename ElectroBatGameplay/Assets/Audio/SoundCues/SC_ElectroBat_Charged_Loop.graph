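# Sound cue for the ElectroBat's charged loop, branching on IsLocal and then IsEnemy.
# The local path mixes the looping bed, an activation sound and looping, randomly delayed one-shots with pitch and volume variation, the latter on only two of four quality levels.
# Non-local paths play just the bed through close-range melee attenuation, with enemy and non-enemy branches configured identically.
# The cue overrides concurrency to cap simultaneous instances.
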SoundCue /ElectroBatGameplay/Assets/Audio/SoundCues/SC_ElectroBat_Charged_Loop
root=Branch_0
nodes [
 {
  id=Attenuation_0
  children [Modulator_7]
  properties {attenuationSettings=/Game/Sounds/Attenuation/Fort_Weapon_Melee_Attack_Close_Range_Mono_Att}
 }
 {
  id=Attenuation_1
  children [Modulator_8]
  properties {attenuationSettings=&Fort_Weapon_Melee_Attack_Close_Range_Mono_Att}
 }
 {id=Attenuation_2, children [Mixer_5]}
 {id=Branch_0, children [Attenuation_2, Branch_1, Branch_1], properties {boolParameterName=IsLocal}}
 {id=Branch_1, children [Attenuation_0, Attenuation_1, Attenuation_0], properties {boolParameterName=IsEnemy}}
 {id=Delay_0, children [Modulator_6], properties {delayMax=0.75}}
 {id=Looping_0, children [Delay_0]}
 {id=Mixer_5, children [Modulator_0, QualityLevel_0, WavePlayer_23], properties {inputVolume [2,1.2,1]}}
 {id=Modulator_0, children [WavePlayer_3], properties {pitchMax=1, pitchMin=1, volumeMax=1, volumeMin=1}}
 {id=Modulator_6, children [Random_0], properties {pitchMax=1.2, pitchMin=0.8, volumeMax=1, volumeMin=0.3}}
 {id=Modulator_7, children [WavePlayer_21], properties {pitchMax=1, pitchMin=1, volumeMax=1, volumeMin=1}}
 {id=Modulator_8, children [WavePlayer_22], properties {pitchMax=1, pitchMin=1, volumeMax=1, volumeMin=1}}
 {id=QualityLevel_0, children [Looping_0,Looping_0,null,null]}
 {
  id=Random_0
  children [WavePlayer_13, WavePlayer_14, WavePlayer_15, WavePlayer_16, WavePlayer_17, WavePlayer_18, WavePlayer_19, WavePlayer_20]
  properties {weights [1,1,1,1,1,1,1,1]}
 }
 {id=WavePlayer_13, properties {soundWaveAssetPtr=/ElectroBatGameplay/Assets/Audio/Wavefiles/ElectroBat_ChargedLoop_OneShots_01}}
 {id=WavePlayer_14, properties {soundWaveAssetPtr=/ElectroBatGameplay/Assets/Audio/Wavefiles/ElectroBat_ChargedLoop_OneShots_02}}
 {id=WavePlayer_15, properties {soundWaveAssetPtr=/ElectroBatGameplay/Assets/Audio/Wavefiles/ElectroBat_ChargedLoop_OneShots_03}}
 {id=WavePlayer_16, properties {soundWaveAssetPtr=/ElectroBatGameplay/Assets/Audio/Wavefiles/ElectroBat_ChargedLoop_OneShots_04}}
 {id=WavePlayer_17, properties {soundWaveAssetPtr=/ElectroBatGameplay/Assets/Audio/Wavefiles/ElectroBat_ChargedLoop_OneShots_05}}
 {id=WavePlayer_18, properties {soundWaveAssetPtr=/ElectroBatGameplay/Assets/Audio/Wavefiles/ElectroBat_ChargedLoop_OneShots_06}}
 {id=WavePlayer_19, properties {soundWaveAssetPtr=/ElectroBatGameplay/Assets/Audio/Wavefiles/ElectroBat_ChargedLoop_OneShots_07}}
 {id=WavePlayer_20, properties {soundWaveAssetPtr=/ElectroBatGameplay/Assets/Audio/Wavefiles/ElectroBat_ChargedLoop_OneShots_08}}
 {id=WavePlayer_21, properties {bLooping=true, soundWaveAssetPtr=/ElectroBatGameplay/Assets/Audio/Wavefiles/ElectroBat_ChargedLoop}}
 {id=WavePlayer_22, properties {bLooping=true, soundWaveAssetPtr=&ElectroBat_ChargedLoop}}
 {id=WavePlayer_23, properties {soundWaveAssetPtr=/ElectroBatGameplay/Assets/Audio/Wavefiles/ElectroBat_ChargedLoop_Activate}}
 {id=WavePlayer_3, properties {bLooping=true, soundWaveAssetPtr=&ElectroBat_ChargedLoop}}
]
properties {
 bOverrideConcurrency=true
 maxDistance=2097152
 soundClassObject=&Fort_WeaponMelee
 volumeMultiplier=0.936
 concurrencyOverrides {maxCount=3}
}
