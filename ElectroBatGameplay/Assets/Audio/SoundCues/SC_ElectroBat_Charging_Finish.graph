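# Sound cue for the finishing stage of the ElectroBat melee weapon's charge.
# After a fixed delay, it layers the Charging_C wave with one of four equally weighted random Charging_D variations.
# The root mixer's first input is empty, so only the delayed branch produces sound.
# It uses the melee weapon sound class, a close-range stereo melee attenuation asset and a concurrency override capping simultaneous instances.
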SoundCue /ElectroBatGameplay/Assets/Audio/SoundCues/SC_ElectroBat_Charging_Finish
root=Mixer_2
nodes [
 {id=Delay_0, children [Mixer_1], properties {delayMax=0.7, delayMin=0.7}}
 {id=Mixer_1, children [WavePlayer_1,Random_0], properties {inputVolume [1,1]}}
 {id=Mixer_2, children [null,Delay_0], properties {inputVolume [1,1]}}
 {id=Random_0, children [WavePlayer_2,WavePlayer_7,WavePlayer_8,WavePlayer_9], properties {weights [1,1,1,1]}}
 {id=WavePlayer_1, properties {soundWaveAssetPtr=/ElectroBatGameplay/Assets/Audio/Wavefiles/ElectroBat_Charging_C}}
 {id=WavePlayer_2, properties {soundWaveAssetPtr=/ElectroBatGameplay/Assets/Audio/Wavefiles/ElectroBat_Charging_D_01}}
 {id=WavePlayer_7, properties {soundWaveAssetPtr=/ElectroBatGameplay/Assets/Audio/Wavefiles/ElectroBat_Charging_D_02}}
 {id=WavePlayer_8, properties {soundWaveAssetPtr=/ElectroBatGameplay/Assets/Audio/Wavefiles/ElectroBat_Charging_D_03}}
 {id=WavePlayer_9, properties {soundWaveAssetPtr=/ElectroBatGameplay/Assets/Audio/Wavefiles/ElectroBat_Charging_D_04}}
]
properties {
 attenuationSettings=/Game/Sounds/Attenuation/Fort_Weapon_Melee_Attack_Close_Range_Stereo_Att
 bOverrideConcurrency=true
 maxDistance=10058.75
 soundClassObject=&Fort_WeaponMelee
 volumeMultiplier=0.936
 attenuationOverrides {distanceAlgorithm=NaturalSound, falloffDistance=4000, attenuationShapeExtents {x=1000}}
 concurrencyOverrides {maxCount=3}
}
