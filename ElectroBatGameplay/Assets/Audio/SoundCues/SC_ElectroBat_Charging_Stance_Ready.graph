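# Sound cue for the ElectroBat charging stance ready sound, reusing the weapon's three Sprint Start waves at lowered pitch and volume.
# An equal-weight random pick feeds a default modulator and single-input mixer, then a close-range stereo melee attenuation node.
# The random node is excluded from branch culling.
# It plays in the melee weapon sound class and references the PickupMaximum_Oldest concurrency set.
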SoundCue /ElectroBatGameplay/Assets/Audio/SoundCues/SC_ElectroBat_Charging_Stance_Ready
root=Attenuation_0
nodes [
 {
  id=Attenuation_0
  children [Mixer_1]
  properties {attenuationSettings=/Game/Sounds/Attenuation/Fort_Weapon_Melee_Attack_Close_Range_Stereo_Att}
 }
 {id=Mixer_1, children [Modulator_2], properties {inputVolume [1]}}
 {id=Modulator_2, children [Random_1]}
 {id=Random_1, children [WavePlayer_0,WavePlayer_5,WavePlayer_6], properties {bShouldExcludeFromBranchCulling=true, weights [1,1,1]}}
 {id=WavePlayer_0, properties {soundWaveAssetPtr=/ElectroBatGameplay/Assets/Audio/Wavefiles/ElectroBat_Sprint_Start_01}}
 {id=WavePlayer_5, properties {soundWaveAssetPtr=/ElectroBatGameplay/Assets/Audio/Wavefiles/ElectroBat_Sprint_Start_02}}
 {id=WavePlayer_6, properties {soundWaveAssetPtr=/ElectroBatGameplay/Assets/Audio/Wavefiles/ElectroBat_Sprint_Start_03}}
]
properties {
 maxDistance=10058.75
 pitchMultiplier=0.8
 soundClassObject=&Fort_WeaponMelee
 volumeMultiplier=0.3
 concurrencyOverrides {maxCount=1, resolutionRule=PreventNew}
 concurrencySet [&PickupMaximum_Oldest]
}
