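# Sound cue for the start of the ElectroBat's charging sound, playing the lead-in wave of its charged loop.
# A single wave player feeds the output directly, with a slight volume boost under the melee weapon sound class.
# It references a close-range stereo melee attenuation asset, carries natural-sound falloff override values, and overrides concurrency to cap simultaneous instances.
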SoundCue /ElectroBatGameplay/Assets/Audio/SoundCues/SC_ElectroBat_Charging_Start
root=WavePlayer_1
nodes [{id=WavePlayer_1, properties {soundWaveAssetPtr=/ElectroBatGameplay/Assets/Audio/Wavefiles/ElectroBat_ChargedLoop_Lead_In}}]
properties {
 attenuationSettings=/Game/Sounds/Attenuation/Fort_Weapon_Melee_Attack_Close_Range_Stereo_Att
 bOverrideConcurrency=true
 maxDistance=10058.75
 soundClassObject=&Fort_WeaponMelee
 volumeMultiplier=1.25
 attenuationOverrides {distanceAlgorithm=NaturalSound, falloffDistance=4000, attenuationShapeExtents {x=1000}}
 concurrencyOverrides {maxCount=3}
}
