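# Sound cue for dropping the ElectroBat, using weapon-drop attenuation and the loot-collect sound class.
# A Mixer layers a fixed drop sample with a randomly chosen, pitch-lowered swing whoosh and a quieter random spark sample.
# A concurrency override allows only one instance at a time and blocks new plays while it is active.
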SoundCue /ElectroBatGameplay/Assets/Audio/SoundCues/SC_ElectroBat_Drop
root=Mixer_0
nodes [
 {id=Mixer_0, children [WavePlayer_4,Modulator_0,Random_3], properties {inputVolume [1,0.7,0.5]}}
 {id=Modulator_0, children [Random_2], properties {pitchMax=0.95, pitchMin=0.85}}
 {id=Random_2, children [WavePlayer_10, WavePlayer_11, WavePlayer_12, WavePlayer_13, WavePlayer_14], properties {weights [1,1,1,1,1]}}
 {id=Random_3, children [WavePlayer_15, WavePlayer_16, WavePlayer_17, WavePlayer_18], properties {weights [1,1,1,1]}}
 {id=WavePlayer_10, properties {soundWaveAssetPtr=/ElectroBatGameplay/Assets/Audio/Wavefiles/SC_ElectroBat_Swing_Whoosh_A_01}}
 {id=WavePlayer_11, properties {soundWaveAssetPtr=/ElectroBatGameplay/Assets/Audio/Wavefiles/SC_ElectroBat_Swing_Whoosh_A_02}}
 {id=WavePlayer_12, properties {soundWaveAssetPtr=/ElectroBatGameplay/Assets/Audio/Wavefiles/SC_ElectroBat_Swing_Whoosh_A_03}}
 {id=WavePlayer_13, properties {soundWaveAssetPtr=/ElectroBatGameplay/Assets/Audio/Wavefiles/SC_ElectroBat_Swing_Whoosh_A_04}}
 {id=WavePlayer_14, properties {soundWaveAssetPtr=/ElectroBatGameplay/Assets/Audio/Wavefiles/SC_ElectroBat_Swing_Whoosh_A_05}}
 {id=WavePlayer_15, properties {soundWaveAssetPtr=/ElectroBatGameplay/Assets/Audio/Wavefiles/ElectroBat_Sparks_B_04}}
 {id=WavePlayer_16, properties {soundWaveAssetPtr=/ElectroBatGameplay/Assets/Audio/Wavefiles/ElectroBat_Sparks_B_03}}
 {id=WavePlayer_17, properties {soundWaveAssetPtr=/ElectroBatGameplay/Assets/Audio/Wavefiles/ElectroBat_Sparks_B_02}}
 {id=WavePlayer_18, properties {soundWaveAssetPtr=/ElectroBatGameplay/Assets/Audio/Wavefiles/ElectroBat_Sparks_B_01}}
 {id=WavePlayer_4, properties {soundWaveAssetPtr=/ElectroBatGameplay/Assets/Audio/Wavefiles/ElectroBat_Drop}}
]
properties {
 attenuationSettings=&Fort_Weapon_Drop_Att
 bOverrideConcurrency=true
 maxDistance=3000
 soundClassObject=&Fort_Loot_Collect
 volumeMultiplier=0.45
 concurrencyOverrides {maxCount=1, resolutionRule=PreventNew}
}
