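# Sound cue for the foley heard when equipping the ElectroBat weapon, layering two grab sounds with an electric spark.
# A mixer blends three branches at different volumes, each offset by a short fixed delay so the grabs and spark land in a staggered sequence.
# The second grab and the randomly chosen one of four spark variants get slight pitch and volume modulation.
# It uses weapon foley class, reload attenuation and a concurrency limit override.
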SoundCue /ElectroBatGameplay/Assets/Audio/SoundCues/SC_ElectroBat_Equip_Foley
root=Mixer_0
nodes [
 {id=Delay_0, children [Modulator_1], properties {delayMax=0.3, delayMin=0.3}}
 {id=Delay_1, children [WavePlayer_0], properties {delayMax=0.2, delayMin=0.2}}
 {id=Delay_2, children [Modulator_0], properties {delayMax=0.22, delayMin=0.22}}
 {id=Mixer_0, children [Delay_1,Delay_0,Delay_2], properties {inputVolume [1.5,0.6,1]}}
 {id=Modulator_0, children [Random_1], properties {pitchMax=1.1, volumeMax=1, volumeMin=0.8}}
 {id=Modulator_1, children [WavePlayer_1], properties {pitchMax=1, pitchMin=0.9, volumeMax=1, volumeMin=0.7}}
 {
  id=Random_1
  children [WavePlayer_7, WavePlayer_8, WavePlayer_9, WavePlayer_10]
  properties {bShouldExcludeFromBranchCulling=true, weights [1,1,1,1]}
 }
 {id=WavePlayer_0, properties {soundWaveAssetPtr=/ElectroBatGameplay/Assets/Audio/Wavefiles/Electrobat_Foley_Grab_A}}
 {id=WavePlayer_1, properties {soundWaveAssetPtr=/ElectroBatGameplay/Assets/Audio/Wavefiles/ElectroBat_Foley_Grab_B}}
 {id=WavePlayer_10, properties {soundWaveAssetPtr=/ElectroBatGameplay/Assets/Audio/Wavefiles/ElectroBat_Equip_Spark_04}}
 {id=WavePlayer_7, properties {soundWaveAssetPtr=/ElectroBatGameplay/Assets/Audio/Wavefiles/ElectroBat_Equip_Spark_01}}
 {id=WavePlayer_8, properties {soundWaveAssetPtr=/ElectroBatGameplay/Assets/Audio/Wavefiles/ElectroBat_Equip_Spark_02}}
 {id=WavePlayer_9, properties {soundWaveAssetPtr=/ElectroBatGameplay/Assets/Audio/Wavefiles/ElectroBat_Equip_Spark_03}}
]
properties {
 attenuationSettings=&Fort_Weapon_Reload_Att
 bOverrideConcurrency=true
 maxDistance=2250
 soundClassObject=&Fort_WeaponFoley
 volumeMultiplier=1
 concurrencyOverrides {maxCount=3}
}
