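# Default hit impact sound cue for the ElectroBat, randomly playing one of four equally weighted medium hit waves on the melee weapon sound class.
# An islocal branch sends local or unset hits through an attenuation node without a settings asset, and others to an IsEnemy branch.
# That branch uses pickaxe impact attenuation, halving volume for non-enemy hits.
# Global and local impact concurrency sets are assigned to the cue.
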SoundCue /ElectroBatGameplay/Assets/Audio/SoundCues/SC_ElectroBat_Hit_Default
root=Branch_2
nodes [
 {id=Attenuation_2, children [Modulator_2]}
 {id=Attenuation_3, children [Modulator_3], properties {attenuationSettings=/Game/Sounds/Attenuation/Fort_Weapon_Pickaxe_Player_Impact_Att}}
 {id=Attenuation_4, children [Modulator_4], properties {attenuationSettings=&Fort_Weapon_Pickaxe_Player_Impact_Att}}
 {id=Branch_2, children [Attenuation_2, Branch_3, Attenuation_2], properties {boolParameterName=islocal}}
 {id=Branch_3, children [Attenuation_4, Attenuation_3, Attenuation_4], properties {boolParameterName=IsEnemy}}
 {id=Modulator_2, children [Random_3], properties {pitchMax=1, pitchMin=1, volumeMax=1, volumeMin=1}}
 {id=Modulator_3, children [Random_3], properties {pitchMax=1, pitchMin=1, volumeMax=0.5, volumeMin=0.5}}
 {id=Modulator_4, children [Random_3], properties {pitchMax=1, pitchMin=1, volumeMax=1, volumeMin=1}}
 {
  id=Random_3
  children [WavePlayer_6, WavePlayer_11, WavePlayer_12, WavePlayer_13]
  properties {bShouldExcludeFromBranchCulling=true, weights [1,1,1,1]}
 }
 {id=WavePlayer_11, properties {soundWaveAssetPtr=/ElectroBatGameplay/Assets/Audio/Wavefiles/ElectroBat_Hit_Medium_02}}
 {id=WavePlayer_12, properties {soundWaveAssetPtr=/ElectroBatGameplay/Assets/Audio/Wavefiles/ElectroBat_Hit_Medium_03}}
 {id=WavePlayer_13, properties {soundWaveAssetPtr=/ElectroBatGameplay/Assets/Audio/Wavefiles/ElectroBat_Hit_Medium_04}}
 {id=WavePlayer_6, properties {soundWaveAssetPtr=/ElectroBatGameplay/Assets/Audio/Wavefiles/ElectroBat_Hit_Medium_01}}
]
properties {
 maxDistance=2097152
 soundClassObject=&Fort_WeaponMelee
 virtualizationMode=Disabled
 volumeMultiplier=0.5
 attenuationOverrides {distanceAlgorithm=NaturalSound, falloffDistance=4000, attenuationShapeExtents {x=1000}}
 concurrencyOverrides {maxCount=1, retriggerTime=0.05}
 concurrencySet [/ElectroBatGameplay/Assets/Audio/CONC_ElectroBat_Impact_Global, /ElectroBatGameplay/Assets/Audio/CONC_ElectroBat_Impact_Local]
}
