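# Sound cue for the first-player (1P) ElectroBat homerun hit, in the melee weapon sound class.
# A mixer layers a boosted, pitch-varied homerun sting, a random bat crack, a briefly delayed random crowd reaction and a modulated random heavy impact.
# Bat and crowd layers sit lower in the mix, and all random nodes are excluded from branch culling.
# Global and local ElectroBat impact concurrency sets govern overlapping plays, with virtualization disabled.
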SoundCue /ElectroBatGameplay/Assets/Audio/SoundCues/SC_ElectroBat_Hit_Homerun_1P
root=Mixer_2
nodes [
 {id=Delay_0, children [Random_3], properties {delayMax=0.1, delayMin=0.1}}
 {id=Mixer_1, children [Modulator_1,Random_2,Delay_0,Modulator_2], properties {inputVolume [1.5,0.5,0.5,1]}}
 {id=Mixer_2, children [Mixer_1], properties {inputVolume [1]}}
 {id=Modulator_1, children [WavePlayer_7], properties {pitchMax=1.1, pitchMin=0.9, volumeMax=1, volumeMin=1}}
 {id=Modulator_2, children [Random_4]}
 {id=Random_2, children [WavePlayer_8,WavePlayer_9], properties {bShouldExcludeFromBranchCulling=true, weights [1,1]}}
 {id=Random_3, children [WavePlayer_10, WavePlayer_11], properties {bShouldExcludeFromBranchCulling=true, weights [1,1]}}
 {
  id=Random_4
  children [WavePlayer_6, WavePlayer_12, WavePlayer_13, WavePlayer_14]
  properties {bShouldExcludeFromBranchCulling=true, weights [1,1,1,1]}
 }
 {id=WavePlayer_10, properties {soundWaveAssetPtr=/ElectroBatGameplay/Assets/Audio/Wavefiles/ElectroBat_Fourth_Hit_Homerun_Crowd01}}
 {id=WavePlayer_11, properties {soundWaveAssetPtr=/ElectroBatGameplay/Assets/Audio/Wavefiles/ElectroBat_Fourth_Hit_Homerun_Crowd02}}
 {id=WavePlayer_12, properties {soundWaveAssetPtr=/ElectroBatGameplay/Assets/Audio/Wavefiles/ElectroBat_Hit_Heavy_02}}
 {id=WavePlayer_13, properties {soundWaveAssetPtr=/ElectroBatGameplay/Assets/Audio/Wavefiles/ElectroBat_Hit_Heavy_03}}
 {id=WavePlayer_14, properties {soundWaveAssetPtr=/ElectroBatGameplay/Assets/Audio/Wavefiles/ElectroBat_Hit_Heavy_04}}
 {id=WavePlayer_6, properties {soundWaveAssetPtr=/ElectroBatGameplay/Assets/Audio/Wavefiles/ElectroBat_Hit_Heavy_01}}
 {id=WavePlayer_7, properties {soundWaveAssetPtr=/ElectroBatGameplay/Assets/Audio/Wavefiles/ElectroBat_Hit_Homerun_01}}
 {id=WavePlayer_8, properties {soundWaveAssetPtr=/ElectroBatGameplay/Assets/Audio/Wavefiles/ElectroBat_Fourth_Hit_Homerun_Bat02}}
 {id=WavePlayer_9, properties {soundWaveAssetPtr=/ElectroBatGameplay/Assets/Audio/Wavefiles/ElectroBat_Fourth_Hit_Homerun_Bat03}}
]
properties {
 maxDistance=2097152
 soundClassObject=&Fort_WeaponMelee
 virtualizationMode=Disabled
 attenuationOverrides {distanceAlgorithm=NaturalSound, falloffDistance=4000, attenuationShapeExtents {x=1000}}
 concurrencyOverrides {maxCount=3}
 concurrencySet [/ElectroBatGameplay/Assets/Audio/CONC_ElectroBat_Impact_Global, /ElectroBatGameplay/Assets/Audio/CONC_ElectroBat_Impact_Local]
}
