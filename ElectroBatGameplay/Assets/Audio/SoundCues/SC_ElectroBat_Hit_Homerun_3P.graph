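# Sound cue for the third-person Electro Bat homerun hit, using "Launched" hit recordings in the melee weapon sound class.
# A random node picks one of three equally weighted wave variations, and a modulator above it applies slight random pitch variation at constant volume.
# It uses close-range stereo melee attenuation, overrides concurrency to cap simultaneous instances, and disables virtualization.
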SoundCue /ElectroBatGameplay/Assets/Audio/SoundCues/SC_ElectroBat_Hit_Homerun_3P
root=Modulator_3
nodes [
 {id=Modulator_3, children [Random_0], properties {pitchMax=1.15, pitchMin=0.9, volumeMax=1, volumeMin=1}}
 {id=Random_0, children [WavePlayer_0,WavePlayer_1,WavePlayer_2], properties {bShouldExcludeFromBranchCulling=true, weights [1,1,1]}}
 {id=WavePlayer_0, properties {soundWaveAssetPtr=/ElectroBatGameplay/Assets/Audio/Wavefiles/ElectroBat_Hit_Homerun_Launched_3P_01}}
 {id=WavePlayer_1, properties {soundWaveAssetPtr=/ElectroBatGameplay/Assets/Audio/Wavefiles/ElectroBat_Hit_Homerun_Launched_3P_02}}
 {id=WavePlayer_2, properties {soundWaveAssetPtr=/ElectroBatGameplay/Assets/Audio/Wavefiles/ElectroBat_Hit_Homerun_Launched_3P_03}}
]
properties {
 attenuationSettings=/Game/Sounds/Attenuation/Fort_Weapon_Melee_Attack_Close_Range_Stereo_Att
 bOverrideConcurrency=true
 maxDistance=10058.75
 soundClassObject=&Fort_WeaponMelee
 virtualizationMode=Disabled
 volumeMultiplier=1
 attenuationOverrides {distanceAlgorithm=NaturalSound, falloffDistance=4000, attenuationShapeExtents {x=1000}}
 concurrencyOverrides {maxCount=2}
}
