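# Sound cue for the first-person impact of an uncharged ElectroBat melee hit, routed to the melee weapon sound class.
# A Mixer layers a random heavy-hit wave at full volume with a random quieter "Hit A" wave.
# The mix passes through a Modulator pinned to unity pitch and volume, then an Attenuation node.
# Global and local impact concurrency sets govern simultaneous playback, and virtualization is disabled.
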SoundCue /ElectroBatGameplay/Assets/Audio/SoundCues/SC_ElectroBat_Hit_Uncharged_1P
root=Attenuation_2
nodes [
 {id=Attenuation_2, children [Modulator_2]}
 {id=Mixer_1, children [Random_1,Random_0], properties {inputVolume [1,0.3]}}
 {id=Modulator_2, children [Mixer_1], properties {pitchMax=1, pitchMin=1, volumeMax=1, volumeMin=1}}
 {id=Random_0, children [WavePlayer_0,WavePlayer_1,WavePlayer_2], properties {weights [1,1,1]}}
 {
  id=Random_1
  children [WavePlayer_3,WavePlayer_4,WavePlayer_5,WavePlayer_6]
  properties {bShouldExcludeFromBranchCulling=true, weights [1,1,1,1]}
 }
 {id=WavePlayer_0, properties {soundWaveAssetPtr=/ElectroBatGameplay/Assets/Audio/Wavefiles/ElectroBat_Hit_A_01}}
 {id=WavePlayer_1, properties {soundWaveAssetPtr=/ElectroBatGameplay/Assets/Audio/Wavefiles/ElectroBat_Hit_A_02}}
 {id=WavePlayer_2, properties {soundWaveAssetPtr=/ElectroBatGameplay/Assets/Audio/Wavefiles/ElectroBat_Hit_A_03}}
 {id=WavePlayer_3, properties {soundWaveAssetPtr=/ElectroBatGameplay/Assets/Audio/Wavefiles/ElectroBat_Hit_Heavy_01}}
 {id=WavePlayer_4, properties {soundWaveAssetPtr=/ElectroBatGameplay/Assets/Audio/Wavefiles/ElectroBat_Hit_Heavy_02}}
 {id=WavePlayer_5, properties {soundWaveAssetPtr=/ElectroBatGameplay/Assets/Audio/Wavefiles/ElectroBat_Hit_Heavy_03}}
 {id=WavePlayer_6, properties {soundWaveAssetPtr=/ElectroBatGameplay/Assets/Audio/Wavefiles/ElectroBat_Hit_Heavy_04}}
]
properties {
 maxDistance=2097152
 soundClassObject=&Fort_WeaponMelee
 virtualizationMode=Disabled
 volumeMultiplier=1.25
 attenuationOverrides {distanceAlgorithm=NaturalSound, falloffDistance=4000, attenuationShapeExtents {x=1000}}
 concurrencyOverrides {maxCount=1, retriggerTime=0.05}
 concurrencySet [/ElectroBatGameplay/Assets/Audio/CONC_ElectroBat_Impact_Global, /ElectroBatGameplay/Assets/Audio/CONC_ElectroBat_Impact_Local]
}
