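# Sound cue for the third-person impact sound of an uncharged ElectroBat melee hit.
# A random node picks one of four equally weighted heavy hit waves and is excluded from branch culling.
# It plays in the melee weapon sound class with a close-range mono melee attenuation asset.
# Concurrency overrides cap simultaneous plays and retrigger rate, and virtualization is disabled.
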SoundCue /ElectroBatGameplay/Assets/Audio/SoundCues/SC_ElectroBat_Hit_Uncharged_3P
root=Random_1
nodes [
 {
  id=Random_1
  children [WavePlayer_3,WavePlayer_4,WavePlayer_5,WavePlayer_6]
  properties {bShouldExcludeFromBranchCulling=true, weights [1,1,1,1]}
 }
 {id=WavePlayer_3, properties {soundWaveAssetPtr=/ElectroBatGameplay/Assets/Audio/Wavefiles/ElectroBat_Hit_Heavy_01}}
 {id=WavePlayer_4, properties {soundWaveAssetPtr=/ElectroBatGameplay/Assets/Audio/Wavefiles/ElectroBat_Hit_Heavy_02}}
 {id=WavePlayer_5, properties {soundWaveAssetPtr=/ElectroBatGameplay/Assets/Audio/Wavefiles/ElectroBat_Hit_Heavy_03}}
 {id=WavePlayer_6, properties {soundWaveAssetPtr=/ElectroBatGameplay/Assets/Audio/Wavefiles/ElectroBat_Hit_Heavy_04}}
]
properties {
 attenuationSettings=/Game/Sounds/Attenuation/Fort_Weapon_Melee_Attack_Close_Range_Mono_Att
 bOverrideConcurrency=true
 maxDistance=10058.75
 soundClassObject=&Fort_WeaponMelee
 virtualizationMode=Disabled
 volumeMultiplier=1
 attenuationOverrides {distanceAlgorithm=NaturalSound, falloffDistance=4000, attenuationShapeExtents {x=1000}}
 concurrencyOverrides {maxCount=2, retriggerTime=0.05}
}
