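# Sound cue for Electro Bat world impact hits, playing one of four equally weighted medium hit waves on the melee weapon sound class.
# An IsLocal branch sends true to its own attenuation node, otherwise to an IsEnemy branch whose outcomes all share a pickaxe-impact attenuation and mixer.
# Both paths feed the same random node, which is excluded from branch culling.
# Global and local impact concurrency sets are assigned, and virtualization is disabled.
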SoundCue /ElectroBatGameplay/Assets/Audio/SoundCues/SC_ElectroBat_Hit_World
root=Branch_0
nodes [
 {id=Attenuation_0, children [Random_3]}
 {id=Attenuation_1, children [Mixer_1], properties {attenuationSettings=/Game/Sounds/Attenuation/Fort_Weapon_Pickaxe_Player_Impact_Att}}
 {id=Branch_0, children [Attenuation_0, Branch_1, Branch_1], properties {boolParameterName=IsLocal}}
 {id=Branch_1, children [Attenuation_1, Attenuation_1, Attenuation_1], properties {boolParameterName=IsEnemy}}
 {id=Mixer_1, children [Random_3], properties {inputVolume [1]}}
 {
  id=Random_3
  children [WavePlayer_6, WavePlayer_11, WavePlayer_12, WavePlayer_13]
  properties {bShouldExcludeFromBranchCulling=true, weights [1,1,1,1]}
 }
 {id=WavePlayer_11, properties {soundWaveAssetPtr=/ElectroBatGameplay/Assets/Audio/Wavefiles/ElectroBat_Hit_Medium_02}}
 {id=WavePlayer_12, properties {soundWaveAssetPtr=/ElectroBatGameplay/Assets/Audio/Wavefiles/ElectroBat_Hit_Medium_03}}
 {id=WavePlayer_13, properties {soundWaveAssetPtr=/ElectroBatGameplay/Assets/Audio/Wavefiles/ElectroBat_Hit_Medium_04}}
 {id=WavePlayer_6, properties {soundWaveAssetPtr=/ElectroBatGameplay/Assets/Audio/Wavefiles/ElectroBat_Hit_Medium_01}}
]
properties {
 attenuationSettings=&Fort_Weapon_Pickaxe_Player_Impact_Att
 maxDistance=2097152
 soundClassObject=&Fort_WeaponMelee
 virtualizationMode=Disabled
 volumeMultiplier=1
 attenuationOverrides {distanceAlgorithm=NaturalSound, falloffDistance=4000, attenuationShapeExtents {x=1000}}
 concurrencyOverrides {maxCount=1, retriggerTime=0.05}
 concurrencySet [/ElectroBatGameplay/Assets/Audio/CONC_ElectroBat_Impact_Global, /ElectroBatGameplay/Assets/Audio/CONC_ElectroBat_Impact_Local]
}
