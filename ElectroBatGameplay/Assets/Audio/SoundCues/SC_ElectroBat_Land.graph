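# Sound cue for the ElectroBat landing, blending its dedicated land sound with a rock impact.
# A Mixer layers the land wave at higher volume over a quieter rock impact, which a short fixed Delay offsets slightly behind it.
# A root Modulator with default settings follows the mix.
# Playback uses weapon-drop attenuation, the loot-collect sound class and a four-voice concurrency limit.
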SoundCue /ElectroBatGameplay/Assets/Audio/SoundCues/SC_ElectroBat_Land
root=Modulator_1
nodes [
 {id=Delay_0, children [WavePlayer_2], properties {delayMax=0.03, delayMin=0.03}}
 {id=Mixer_0, children [Delay_0,WavePlayer_1], properties {inputVolume [0.4,0.85]}}
 {id=Modulator_1, children [Mixer_0]}
 {id=WavePlayer_1, properties {soundWaveAssetPtr=/ElectroBatGameplay/Assets/Audio/Wavefiles/ElectroBat_Land}}
 {id=WavePlayer_2, properties {soundWaveAssetPtr=/ElectroBatGameplay/Assets/Audio/Wavefiles/rock_impact_06}}
]
properties {
 attenuationSettings=&Fort_Weapon_Drop_Att
 maxDistance=3000
 soundClassObject=&Fort_Loot_Collect
 volumeMultiplier=0.8
 concurrencyOverrides {maxCount=4}
 concurrencySet [/Game/Sounds/Concurrency_Rules/Demo_FourMaximum_StopFurthestOldest_VolumeScale]
}
