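# Sound cue for the ElectroBat pickup sound, judging by its name and loot-collect sound class.
# A single wave player feeds the ElectroBat_Pickup_01 wave straight to the output, with no randomization or modulation.
# A shared weapon-pickup attenuation asset shapes falloff.
# It references the PickupMaximum_Oldest concurrency set and also carries single-voice, prevent-new override values.
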SoundCue /ElectroBatGameplay/Assets/Audio/SoundCues/SC_ElectroBat_Pickup
root=WavePlayer_0
nodes [{id=WavePlayer_0, properties {soundWaveAssetPtr=/ElectroBatGameplay/Assets/Audio/Wavefiles/ElectroBat_Pickup_01}}]
properties {
 attenuationSettings=&Fort_Weapon_Pickup_Att
 maxDistance=3500
 soundClassObject=&Fort_Loot_Collect
 volumeMultiplier=1
 concurrencyOverrides {maxCount=1, resolutionRule=PreventNew}
 concurrencySet [&PickupMaximum_Oldest]
}
