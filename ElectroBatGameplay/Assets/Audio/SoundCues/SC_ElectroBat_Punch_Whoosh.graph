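# Sound cue for the ElectroBat punch whoosh, played as a close-range melee weapon sound.
# A Random node picks evenly among five whoosh wave variations and feeds a Modulator at the root.
# It uses the melee weapon sound class and close-range stereo attenuation, and overrides concurrency to cap simultaneous instances.
# It also sends fully to the ElectroBat duck audio bus.
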SoundCue /ElectroBatGameplay/Assets/Audio/SoundCues/SC_ElectroBat_Punch_Whoosh
root=Modulator_0
nodes [
 {id=Modulator_0, children [Random_1]}
 {id=Random_1, children [WavePlayer_0, WavePlayer_4, WavePlayer_8, WavePlayer_9, WavePlayer_15], properties {weights [1,1,1,1,1]}}
 {id=WavePlayer_0, properties {soundWaveAssetPtr=/ElectroBatGameplay/Assets/Audio/Wavefiles/ElectroBat_Punch_Whoosh_01}}
 {id=WavePlayer_15, properties {soundWaveAssetPtr=/ElectroBatGameplay/Assets/Audio/Wavefiles/ElectroBat_Punch_Whoosh_05}}
 {id=WavePlayer_4, properties {soundWaveAssetPtr=/ElectroBatGameplay/Assets/Audio/Wavefiles/ElectroBat_Punch_Whoosh_02}}
 {id=WavePlayer_8, properties {soundWaveAssetPtr=/ElectroBatGameplay/Assets/Audio/Wavefiles/ElectroBat_Punch_Whoosh_03}}
 {id=WavePlayer_9, properties {soundWaveAssetPtr=/ElectroBatGameplay/Assets/Audio/Wavefiles/ElectroBat_Punch_Whoosh_04}}
]
properties {
 attenuationSettings=/Game/Sounds/Attenuation/Fort_Weapon_Melee_Attack_Close_Range_Stereo_Att
 bOverrideConcurrency=true
 maxDistance=10058.75
 soundClassObject=&Fort_WeaponMelee
 volumeMultiplier=0.469
 attenuationOverrides {distanceAlgorithm=NaturalSound, falloffDistance=4000, attenuationShapeExtents {x=1000}}
 busSends [{audioBus=/ElectroBatGameplay/Assets/Audio/AB_ElectroBat_Duck, minSendLevel=0, sendLevel=1, soundSourceBus=null}]
 concurrencyOverrides {maxCount=2}
}
