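# Sound cue for the ElectroBat melee weapon's sprint audio, layering a start one-shot with a looping wind bed.
# A mixer combines one of three randomly chosen, modulated start waves with a quieter, modulated wind loop.
# An enveloper fades the loop in, eases its volume back and raises its pitch over time.
# Output uses the ElectroBat run-loop attenuation, the melee weapon sound class and an overridden concurrency limit.
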SoundCue /ElectroBatGameplay/Assets/Audio/SoundCues/SC_ElectroBat_Sprint_Start_Loop
root=Attenuation_1
nodes [
 {id=Attenuation_1, children [Mixer_7], properties {attenuationSettings=/ElectroBatGameplay/Assets/Audio/ATT_ElectroBat_Run_Loop_Stereo}}
 {
  id=Enveloper_1
  children [Modulator_0]
  properties {
   pitchCurve {editorCurveData {keys [{time=0}, <at 10 1.5>]}}
   volumeCurve {editorCurveData {keys [{arriveTangent=1.046, interpMode=RCIM_Cubic, leaveTangent=1.046, tangentMode=RCTM_User, time=0, value=0}, <at 2 1>, <at 10 0.75>]}}
  }
 }
 {id=Mixer_7, children [Enveloper_1,Modulator_1], properties {inputVolume [0.6,1]}}
 {id=Modulator_0, children [WavePlayer_8], properties {pitchMax=1.1, volumeMax=1, volumeMin=1}}
 {id=Modulator_1, children [Random_0]}
 {id=Random_0, children [WavePlayer_0,WavePlayer_2,WavePlayer_4], properties {bShouldExcludeFromBranchCulling=true, weights [1,1,1]}}
 {id=WavePlayer_0, properties {soundWaveAssetPtr=/ElectroBatGameplay/Assets/Audio/Wavefiles/ElectroBat_Sprint_Start_01}}
 {id=WavePlayer_2, properties {soundWaveAssetPtr=/ElectroBatGameplay/Assets/Audio/Wavefiles/ElectroBat_Sprint_Start_02}}
 {id=WavePlayer_4, properties {soundWaveAssetPtr=/ElectroBatGameplay/Assets/Audio/Wavefiles/ElectroBat_Sprint_Start_03}}
 {id=WavePlayer_8, properties {bLooping=true, soundWaveAssetPtr=/ElectroBatGameplay/Assets/Audio/Wavefiles/ElectroBat_Sprint_Loop_Wind_01}}
]
properties {
 bOverrideConcurrency=true
 maxDistance=6750
 soundClassObject=&Fort_WeaponMelee
 volumeMultiplier=0.25
 concurrencyOverrides {maxCount=3}
}
