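# Sound cue for the large swing shockwave of the ElectroBat melee weapon, on the weapon melee sound class.
# A top random node picks one of two equally weighted modulated branches, one randomizing three large pulse waves and the other three small pulse waves.
# Only one pulse plays per trigger, so small pulses can sound despite the Large name.
# It references a close-range stereo melee attenuation asset and overrides concurrency to cap simultaneous instances.
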SoundCue /ElectroBatGameplay/Assets/Audio/SoundCues/SC_ElectroBat_Swing_Shockwave_Large
root=Random_3
nodes [
 {id=Modulator_0, children [Random_1]}
 {id=Modulator_1, children [Random_2]}
 {id=Random_1, children [WavePlayer_3,WavePlayer_4,WavePlayer_5], properties {bShouldExcludeFromBranchCulling=true, weights [1,1,1]}}
 {id=Random_2, children [WavePlayer_6,WavePlayer_7,WavePlayer_8], properties {bShouldExcludeFromBranchCulling=true, weights [1,1,1]}}
 {id=Random_3, children [Modulator_0,Modulator_1], properties {bShouldExcludeFromBranchCulling=true, weights [1,1]}}
 {id=WavePlayer_3, properties {soundWaveAssetPtr=/ElectroBatGameplay/Assets/Audio/Wavefiles/ElectroBat_Swing_Shockwave_Pulse_Large_01}}
 {id=WavePlayer_4, properties {soundWaveAssetPtr=/ElectroBatGameplay/Assets/Audio/Wavefiles/ElectroBat_Swing_Shockwave_Pulse_Large_02}}
 {id=WavePlayer_5, properties {soundWaveAssetPtr=/ElectroBatGameplay/Assets/Audio/Wavefiles/ElectroBat_Swing_Shockwave_Pulse_Large_03}}
 {id=WavePlayer_6, properties {soundWaveAssetPtr=/ElectroBatGameplay/Assets/Audio/Wavefiles/ElectroBat_Swing_Shockwave_Pulse_Small_01}}
 {id=WavePlayer_7, properties {soundWaveAssetPtr=/ElectroBatGameplay/Assets/Audio/Wavefiles/ElectroBat_Swing_Shockwave_Pulse_Small_02}}
 {id=WavePlayer_8, properties {soundWaveAssetPtr=/ElectroBatGameplay/Assets/Audio/Wavefiles/ElectroBat_Swing_Shockwave_Pulse_Small_03}}
]
properties {
 attenuationSettings=/Game/Sounds/Attenuation/Fort_Weapon_Melee_Attack_Close_Range_Stereo_Att
 bOverrideConcurrency=true
 maxDistance=10058.75
 soundClassObject=&Fort_WeaponMelee
 volumeMultiplier=0.85
 attenuationOverrides {distanceAlgorithm=NaturalSound, falloffDistance=4000, attenuationShapeExtents {x=1000}}
 concurrencyOverrides {maxCount=2}
}
