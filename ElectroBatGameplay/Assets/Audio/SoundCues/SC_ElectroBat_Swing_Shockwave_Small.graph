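# Sound cue for the small shockwave swing of the ElectroBat melee weapon, built from three ElectroBat Hit A wave variations.
# An equally weighted random node picks a wave, which passes through a default-setting modulator into a mixer whose second input is empty.
# It plays at reduced volume in the melee weapon sound class, with a close-range stereo attenuation asset and a concurrency override limiting simultaneous instances.
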SoundCue /ElectroBatGameplay/Assets/Audio/SoundCues/SC_ElectroBat_Swing_Shockwave_Small
root=Mixer_0
nodes [
 {id=Mixer_0, children [Modulator_1,null], properties {inputVolume [1,1]}}
 {id=Modulator_1, children [Random_0]}
 {id=Random_0, children [WavePlayer_0,WavePlayer_1,WavePlayer_2], properties {bShouldExcludeFromBranchCulling=true, weights [1,1,1]}}
 {id=WavePlayer_0, properties {soundWaveAssetPtr=/ElectroBatGameplay/Assets/Audio/Wavefiles/ElectroBat_Hit_A_01}}
 {id=WavePlayer_1, properties {soundWaveAssetPtr=/ElectroBatGameplay/Assets/Audio/Wavefiles/ElectroBat_Hit_A_02}}
 {id=WavePlayer_2, properties {soundWaveAssetPtr=/ElectroBatGameplay/Assets/Audio/Wavefiles/ElectroBat_Hit_A_03}}
]
properties {
 attenuationSettings=/Game/Sounds/Attenuation/Fort_Weapon_Melee_Attack_Close_Range_Stereo_Att
 bOverrideConcurrency=true
 maxDistance=10058.75
 soundClassObject=&Fort_WeaponMelee
 volumeMultiplier=0.6
 attenuationOverrides {distanceAlgorithm=NaturalSound, falloffDistance=4000, attenuationShapeExtents {x=1000}}
 concurrencyOverrides {maxCount=2}
}
