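# Sound cue for the B variant of the ElectroBat's swing spark sound, classed as melee weapon audio.
# A random node picks one of four equally weighted spark waves and feeds a modulator that varies pitch and volume per play.
# It uses close-range melee stereo attenuation and sends to the ElectroBat duck audio bus.
# Global and local swing-spark concurrency sets are assigned to limit overlapping instances.
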SoundCue /ElectroBatGameplay/Assets/Audio/SoundCues/SC_ElectroBat_Swing_Spark_B
root=Modulator_0
nodes [
 {id=Modulator_0, children [Random_0], properties {pitchMax=1.2, volumeMax=1, volumeMin=0.6}}
 {id=Random_0, children [WavePlayer_0,WavePlayer_1,WavePlayer_2,WavePlayer_3], properties {weights [1,1,1,1]}}
 {id=WavePlayer_0, properties {soundWaveAssetPtr=/ElectroBatGameplay/Assets/Audio/Wavefiles/ElectroBat_Sparks_B_01}}
 {id=WavePlayer_1, properties {soundWaveAssetPtr=/ElectroBatGameplay/Assets/Audio/Wavefiles/ElectroBat_Sparks_B_02}}
 {id=WavePlayer_2, properties {soundWaveAssetPtr=/ElectroBatGameplay/Assets/Audio/Wavefiles/ElectroBat_Sparks_B_03}}
 {id=WavePlayer_3, properties {soundWaveAssetPtr=/ElectroBatGameplay/Assets/Audio/Wavefiles/ElectroBat_Sparks_B_04}}
]
properties {
 attenuationSettings=/Game/Sounds/Attenuation/Fort_Weapon_Melee_Attack_Close_Range_Stereo_Att
 maxDistance=10058.75
 soundClassObject=&Fort_WeaponMelee
 volumeMultiplier=0.5
 attenuationOverrides {distanceAlgorithm=NaturalSound, falloffDistance=4000, attenuationShapeExtents {x=1000}}
 busSends [{audioBus=/ElectroBatGameplay/Assets/Audio/AB_ElectroBat_Duck, minSendLevel=0, sendLevel=1, soundSourceBus=null}]
 concurrencyOverrides {maxCount=2}
 concurrencySet [/ElectroBatGameplay/Assets/Audio/CONC_ElectroBat_Swings_Sparks_Global, /ElectroBatGameplay/Assets/Audio/CONC_ElectroBat_Swings_Sparks_Local]
}
