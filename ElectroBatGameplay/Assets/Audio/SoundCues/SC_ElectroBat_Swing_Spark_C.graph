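# Sound cue for the ElectroBat melee weapon's swing, layering electric sparks with a shockwave pulse.
# A mixer blends one of three heavy spark waves, varied in pitch and volume, with a modulated pulse picked at random from large or small sets of three.
# It uses the melee weapon sound class and close-range attenuation, sends to an ElectroBat duck audio bus, and references global and local concurrency limits.
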SoundCue /ElectroBatGameplay/Assets/Audio/SoundCues/SC_ElectroBat_Swing_Spark_C
root=Mixer_0
nodes [
 {id=Mixer_0, children [Modulator_0,Random_3], properties {inputVolume [1,1]}}
 {id=Modulator_0, children [Random_0], properties {pitchMax=1.3, pitchMin=1, volumeMax=1, volumeMin=0.8}}
 {id=Modulator_1, children [Random_1]}
 {id=Modulator_2, children [Random_2]}
 {id=Random_0, children [WavePlayer_0,WavePlayer_1,WavePlayer_2], properties {bShouldExcludeFromBranchCulling=true, weights [1,1,1]}}
 {id=Random_1, children [WavePlayer_3,WavePlayer_4,WavePlayer_5], properties {bShouldExcludeFromBranchCulling=true, weights [1,1,1]}}
 {id=Random_2, children [WavePlayer_6,WavePlayer_7,WavePlayer_8], properties {bShouldExcludeFromBranchCulling=true, weights [1,1,1]}}
 {id=Random_3, children [Modulator_1,Modulator_2], properties {bShouldExcludeFromBranchCulling=true, weights [1,1]}}
 {id=WavePlayer_0, properties {soundWaveAssetPtr=/ElectroBatGameplay/Assets/Audio/Wavefiles/ElectroBat_Sparks_Heavy_01}}
 {id=WavePlayer_1, properties {soundWaveAssetPtr=/ElectroBatGameplay/Assets/Audio/Wavefiles/ElectroBat_Sparks_Heavy_02}}
 {id=WavePlayer_2, properties {soundWaveAssetPtr=/ElectroBatGameplay/Assets/Audio/Wavefiles/ElectroBat_Sparks_Heavy_03}}
 {id=WavePlayer_3, properties {soundWaveAssetPtr=/ElectroBatGameplay/Assets/Audio/Wavefiles/ElectroBat_Swing_Shockwave_Pulse_Large_01}}
 {id=WavePlayer_4, properties {soundWaveAssetPtr=/ElectroBatGameplay/Assets/Audio/Wavefiles/ElectroBat_Swing_Shockwave_Pulse_Large_02}}
 {id=WavePlayer_5, properties {soundWaveAssetPtr=/ElectroBatGameplay/Assets/Audio/Wavefiles/ElectroBat_Swing_Shockwave_Pulse_Large_03}}
 {id=WavePlayer_6, properties {soundWaveAssetPtr=/ElectroBatGameplay/Assets/Audio/Wavefiles/ElectroBat_Swing_Shockwave_Pulse_Small_01}}
 {id=WavePlayer_7, properties {soundWaveAssetPtr=/ElectroBatGameplay/Assets/Audio/Wavefiles/ElectroBat_Swing_Shockwave_Pulse_Small_02}}
 {id=WavePlayer_8, properties {soundWaveAssetPtr=/ElectroBatGameplay/Assets/Audio/Wavefiles/ElectroBat_Swing_Shockwave_Pulse_Small_03}}
]
properties {
 attenuationSettings=/Game/Sounds/Attenuation/Fort_Weapon_Melee_Attack_Close_Range_Stereo_Att
 maxDistance=10058.75
 soundClassObject=&Fort_WeaponMelee
 volumeMultiplier=0.3252
 attenuationOverrides {distanceAlgorithm=NaturalSound, falloffDistance=4000, attenuationShapeExtents {x=1000}}
 busSends [{audioBus=/ElectroBatGameplay/Assets/Audio/AB_ElectroBat_Duck, minSendLevel=0, sendLevel=1, soundSourceBus=null}]
 concurrencyOverrides {maxCount=2}
 concurrencySet [/ElectroBatGameplay/Assets/Audio/CONC_ElectroBat_Swings_Sparks_Global, /ElectroBatGameplay/Assets/Audio/CONC_ElectroBat_Swings_Sparks_Local]
}
