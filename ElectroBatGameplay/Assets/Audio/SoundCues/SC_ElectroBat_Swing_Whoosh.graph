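# Sound cue for the ElectroBat melee swing whoosh, mixed differently for the local player, enemies and other players.
# IsLocal and IsEnemy branches boost remote enemy swings, lower non-enemy ones, and leave local or unset cases unscaled.
# Each path randomly picks one of the same six whoosh waves with slight volume variation.
# It uses close-range melee attenuation and a shared swing concurrency set, and sends to the ElectroBat duck audio bus.
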SoundCue /ElectroBatGameplay/Assets/Audio/SoundCues/SC_ElectroBat_Swing_Whoosh
root=Branch_0
nodes [
 {id=Branch_0, children [Modulator_1,Branch_1,Modulator_1], properties {boolParameterName=IsLocal}}
 {id=Branch_1, children [Mixer_8,Mixer_6,Modulator_1], properties {boolParameterName=IsEnemy}}
 {id=Mixer_6, children [Modulator_3], properties {inputVolume [0.6]}}
 {id=Mixer_8, children [Modulator_2], properties {inputVolume [1.2]}}
 {id=Modulator_1, children [Random_0], properties {volumeMin=0.85}}
 {id=Modulator_2, children [Random_4], properties {volumeMin=0.85}}
 {id=Modulator_3, children [Random_5], properties {volumeMin=0.85}}
 {
  id=Random_0
  children [WavePlayer_1, WavePlayer_2, WavePlayer_3, WavePlayer_4, WavePlayer_5, WavePlayer_10]
  properties {bShouldExcludeFromBranchCulling=true, weights [1,1,1,1,1,1]}
 }
 {
  id=Random_4
  children [WavePlayer_11, WavePlayer_12, WavePlayer_13, WavePlayer_14, WavePlayer_22, WavePlayer_23]
  properties {bShouldExcludeFromBranchCulling=true, weights [1,1,1,1,1,1]}
 }
 {
  id=Random_5
  children [WavePlayer_27, WavePlayer_28, WavePlayer_29, WavePlayer_30, WavePlayer_31, WavePlayer_32]
  properties {bShouldExcludeFromBranchCulling=true, weights [1,1,1,1,1,1]}
 }
 {id=WavePlayer_1, properties {soundWaveAssetPtr=/ElectroBatGameplay/Assets/Audio/Wavefiles/SC_ElectroBat_Swing_Whoosh_01}}
 {id=WavePlayer_10, properties {soundWaveAssetPtr=/ElectroBatGameplay/Assets/Audio/Wavefiles/SC_ElectroBat_Swing_Whoosh_06}}
 {id=WavePlayer_11, properties {soundWaveAssetPtr=&SC_ElectroBat_Swing_Whoosh_01}}
 {id=WavePlayer_12, properties {soundWaveAssetPtr=/ElectroBatGameplay/Assets/Audio/Wavefiles/SC_ElectroBat_Swing_Whoosh_02}}
 {id=WavePlayer_13, properties {soundWaveAssetPtr=/ElectroBatGameplay/Assets/Audio/Wavefiles/SC_ElectroBat_Swing_Whoosh_03}}
 {id=WavePlayer_14, properties {soundWaveAssetPtr=/ElectroBatGameplay/Assets/Audio/Wavefiles/SC_ElectroBat_Swing_Whoosh_04}}
 {id=WavePlayer_2, properties {soundWaveAssetPtr=&SC_ElectroBat_Swing_Whoosh_02}}
 {id=WavePlayer_22, properties {soundWaveAssetPtr=/ElectroBatGameplay/Assets/Audio/Wavefiles/SC_ElectroBat_Swing_Whoosh_05}}
 {id=WavePlayer_23, properties {soundWaveAssetPtr=&SC_ElectroBat_Swing_Whoosh_06}}
 {id=WavePlayer_27, properties {soundWaveAssetPtr=&SC_ElectroBat_Swing_Whoosh_01}}
 {id=WavePlayer_28, properties {soundWaveAssetPtr=&SC_ElectroBat_Swing_Whoosh_02}}
 {id=WavePlayer_29, properties {soundWaveAssetPtr=&SC_ElectroBat_Swing_Whoosh_03}}
 {id=WavePlayer_3, properties {soundWaveAssetPtr=&SC_ElectroBat_Swing_Whoosh_03}}
 {id=WavePlayer_30, properties {soundWaveAssetPtr=&SC_ElectroBat_Swing_Whoosh_04}}
 {id=WavePlayer_31, properties {soundWaveAssetPtr=&SC_ElectroBat_Swing_Whoosh_05}}
 {id=WavePlayer_32, properties {soundWaveAssetPtr=&SC_ElectroBat_Swing_Whoosh_06}}
 {id=WavePlayer_4, properties {soundWaveAssetPtr=&SC_ElectroBat_Swing_Whoosh_04}}
 {id=WavePlayer_5, properties {soundWaveAssetPtr=&SC_ElectroBat_Swing_Whoosh_05}}
]
properties {
 attenuationSettings=/Game/Sounds/Attenuation/Fort_Weapon_Melee_Attack_Close_Range_Stereo_Att
 maxDistance=10058.75
 soundClassObject=&Fort_WeaponMelee
 volumeMultiplier=0.4
 busSends [{
   audioBus=/ElectroBatGameplay/Assets/Audio/AB_ElectroBat_Duck
   maxSendDistance=0
   minSendDistance=1000
   minSendLevel=0
   sendLevel=1
   soundSourceBus=null
  }]
 concurrencyOverrides {maxCount=2}
 concurrencySet [/ElectroBatGameplay/Assets/Audio/CONC_ElectroBat_Swings_Global]
}
