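# Sound cue for an ElectroBat melee swing whoosh, in the melee weapon sound class with close-range melee attenuation.
# A Mixer layers two equally weighted Random picks, one from the B whoosh waves and one from the A set, with A mixed louder despite the cue's B name.
# Both Random nodes are excluded from branch culling.
# The cue overrides concurrency to cap simultaneous instances and sends to the ElectroBat duck audio bus.
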SoundCue /ElectroBatGameplay/Assets/Audio/SoundCues/SC_ElectroBat_Swing_Whoosh_B
root=Mixer_0
nodes [
 {id=Mixer_0, children [Random_0,Random_2], properties {inputVolume [0.3,0.7]}}
 {
  id=Random_0
  children [WavePlayer_1,WavePlayer_2,WavePlayer_3,WavePlayer_5,WavePlayer_6,WavePlayer_7]
  properties {bShouldExcludeFromBranchCulling=true, weights [1,1,1,1,1,1]}
 }
 {
  id=Random_2
  children [WavePlayer_10, WavePlayer_11, WavePlayer_12, WavePlayer_13, WavePlayer_14]
  properties {bShouldExcludeFromBranchCulling=true, weights [1,1,1,1,1]}
 }
 {id=WavePlayer_1, properties {soundWaveAssetPtr=/ElectroBatGameplay/Assets/Audio/Wavefiles/SC_ElectroBat_Swing_Whoosh_B_06}}
 {id=WavePlayer_10, properties {soundWaveAssetPtr=/ElectroBatGameplay/Assets/Audio/Wavefiles/SC_ElectroBat_Swing_Whoosh_A_05}}
 {id=WavePlayer_11, properties {soundWaveAssetPtr=/ElectroBatGameplay/Assets/Audio/Wavefiles/SC_ElectroBat_Swing_Whoosh_A_04}}
 {id=WavePlayer_12, properties {soundWaveAssetPtr=/ElectroBatGameplay/Assets/Audio/Wavefiles/SC_ElectroBat_Swing_Whoosh_A_03}}
 {id=WavePlayer_13, properties {soundWaveAssetPtr=/ElectroBatGameplay/Assets/Audio/Wavefiles/SC_ElectroBat_Swing_Whoosh_A_02}}
 {id=WavePlayer_14, properties {soundWaveAssetPtr=/ElectroBatGameplay/Assets/Audio/Wavefiles/SC_ElectroBat_Swing_Whoosh_A_01}}
 {id=WavePlayer_2, properties {soundWaveAssetPtr=/ElectroBatGameplay/Assets/Audio/Wavefiles/SC_ElectroBat_Swing_Whoosh_B_05}}
 {id=WavePlayer_3, properties {soundWaveAssetPtr=/ElectroBatGameplay/Assets/Audio/Wavefiles/SC_ElectroBat_Swing_Whoosh_B_04}}
 {id=WavePlayer_5, properties {soundWaveAssetPtr=/ElectroBatGameplay/Assets/Audio/Wavefiles/SC_ElectroBat_Swing_Whoosh_B_03}}
 {id=WavePlayer_6, properties {soundWaveAssetPtr=/ElectroBatGameplay/Assets/Audio/Wavefiles/SC_ElectroBat_Swing_Whoosh_B_02}}
 {id=WavePlayer_7, properties {soundWaveAssetPtr=/ElectroBatGameplay/Assets/Audio/Wavefiles/SC_ElectroBat_Swing_Whoosh_B_01}}
]
properties {
 attenuationSettings=/Game/Sounds/Attenuation/Fort_Weapon_Melee_Attack_Close_Range_Stereo_Att
 bOverrideConcurrency=true
 maxDistance=10058.75
 soundClassObject=&Fort_WeaponMelee
 volumeMultiplier=0.5
 attenuationOverrides {distanceAlgorithm=NaturalSound, falloffDistance=4000, attenuationShapeExtents {x=1000}}
 busSends [{audioBus=/ElectroBatGameplay/Assets/Audio/AB_ElectroBat_Duck, minSendLevel=0, sendLevel=1, soundSourceBus=null}]
 concurrencyOverrides {maxCount=2}
}
