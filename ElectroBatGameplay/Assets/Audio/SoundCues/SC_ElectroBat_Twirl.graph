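# Sound cue for the ElectroBat twirl, a melee weapon sound in the Fort_WeaponMelee class.
# A mixer layers one of three equally weighted random swing whooshes with a slightly quieter sprint-charge sound, each through its own modulator for variation.
# The mix feeds a close-range stereo melee attenuation node, and the cue references the PickupMaximum_Oldest concurrency set.
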SoundCue /ElectroBatGameplay/Assets/Audio/SoundCues/SC_ElectroBat_Twirl
root=Attenuation_1
nodes [
 {
  id=Attenuation_1
  children [Mixer_0]
  properties {attenuationSettings=/Game/Sounds/Attenuation/Fort_Weapon_Melee_Attack_Close_Range_Stereo_Att}
 }
 {id=Mixer_0, children [Modulator_0,Modulator_1], properties {inputVolume [1,0.8]}}
 {id=Modulator_0, children [Random_0]}
 {id=Modulator_1, children [WavePlayer_1], properties {volumeMax=1.1}}
 {id=Random_0, children [WavePlayer_2,WavePlayer_3,WavePlayer_4], properties {bShouldExcludeFromBranchCulling=true, weights [1,1,1]}}
 {id=WavePlayer_1, properties {soundWaveAssetPtr=/ElectroBatGameplay/Assets/Audio/Wavefiles/ElectroBat_Sprint_Charge}}
 {id=WavePlayer_2, properties {soundWaveAssetPtr=/ElectroBatGameplay/Assets/Audio/Wavefiles/SC_ElectroBat_Swing_Whoosh_01}}
 {id=WavePlayer_3, properties {soundWaveAssetPtr=/ElectroBatGameplay/Assets/Audio/Wavefiles/SC_ElectroBat_Swing_Whoosh_02}}
 {id=WavePlayer_4, properties {soundWaveAssetPtr=/ElectroBatGameplay/Assets/Audio/Wavefiles/SC_ElectroBat_Swing_Whoosh_03}}
]
properties {
 maxDistance=10058.75
 soundClassObject=&Fort_WeaponMelee
 volumeMultiplier=0.7
 concurrencyOverrides {maxCount=1, resolutionRule=PreventNew}
 concurrencySet [&PickupMaximum_Oldest]
}
